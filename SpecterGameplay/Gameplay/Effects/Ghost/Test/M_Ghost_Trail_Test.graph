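# Test translucent unlit material for a ghost trail effect on Niagara mesh particles.
# Color blends two tints along the V gradient, with a brightened Fresnel edge.
# A panning dissolve texture, shaped by gradient and Fresnel terms, goes through subtract, divide and saturate to form a sharp-edged holes mask.
# A time-driven sine, offset per particle and scaled by the V coordinate, builds an XY world position offset wobble.
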Material /SpecterGameplay/Gameplay/Effects/Ghost/Test/M_Ghost_Trail_Test
graph {
 properties {
  bAutomaticallySetUsageInEditor=false
  bUsedWithNiagaraMeshParticles=true
  blendMode=BLEND_Translucent
  previewMesh=&/Sprite_Ghost
  shadingModel=MSM_Unlit
  usedShadingModels=Unlit
 }
 nodes [
  n1 = OneMinus(None=MaterialFunctionCall() {materialFunction=&LinearGradient}.VGradient)
  n2 = LinearInterpolate(A=Constant4Vector() {constant=<rgba 1.6449 1.0417 2 1>}.RGB, B=Constant4Vector() {constant=<rgba 0.3465 0.078 1 1>}.RGB, Alpha=@n1) {desc=Gradient Color Layer}
  n3 = OneMinus(None=MaterialFunctionCall(Exponent=ScalarParameter() {defaultValue=0.75, desc=Fresnel Edge Exp, parameterName=Fresnel Edge Exp}) {materialFunction=&AxisAlignedFresnel}.Fresnel)
  n4 = ScalarParameter() {defaultValue=3, parameterName=Fresnel Edge Pow}
  LinearInterpolate(
   A=@n2
   B=Multiply(A=@n2, B=ScalarParameter() {defaultValue=10, desc=Edge Intensity, parameterName=Fresnel Edge Intensity})
   Alpha=Saturate(None=Power(Base=@n3, Exp=@n4)) {desc=Fresnel Edge Mask}
  )
  n5 = Time()
  n6 = ParticleRandom()
  n7 = Add(
   A=ComponentMask(None=TextureCoordinate()) {g=true}
   B=Multiply(A=Add(A=@n5, B=@n6), B=ScalarParameter() {defaultValue=1, group=WPO, parameterName=WPOSineSpeed})
  )
  n8 = ScalarParameter() {defaultValue=0.5, group=WPO, parameterName=WPOSineAmount}
  n9 = Multiply(A=Sine(None=Multiply(A=@n7, B=@n8)), B=Constant3Vector() {constant=<rgba 0.75 0.75 0 1>})
  n10 = ScalarParameter() {defaultValue=10, group=WPO, parameterName=WPOSineIntensity}
  n11 = ComponentMask(None=TextureCoordinate()) {g=true}
  NamedRerouteDeclaration(None=Multiply(A=Multiply(A=@n9, B=@n10), B=@n11)) {name=WPO, nodeColor=<rgba 1 0.8203 0 1>}
  OneMinus(None=@n11)
  n12 = VectorParameter() {parameterName=Holes Tile and Pan, defaultValue=<rgba 3 0.75 0 -0.3>}
  n13 = Multiply(A=TextureCoordinate(), B=AppendVector(A=@n12.R, B=@n12.R))
  n14 = AppendVector(A=@n12.B, B=@n12.B)
  n15 = OneMinus(None=TextureSample(UVs=Panner(Coordinate=@n13, Speed=@n14)) {texture=/Game/CharClasses/Outlander/Gadgets/AntiMaterialCharge/Materials/T_AntiMaterialCharge_Dissolve_01}.R)
  n16 = MaterialFunctionCall() {materialFunction=&LinearGradient}
  n17 = Multiply(A=Constant() {r=0.4}, B=Power(Base=OneMinus(None=@n16.VGradient)))
  n18 = MaterialFunctionCall(Exponent=Constant() {r=2}) {materialFunction=&AxisAlignedFresnel}
  n19 = Power(Base=@n1)
  n20 = Constant() {r=0.25}
  n21 = Multiply(
   A=@n19
   B=Power(Base=MaterialFunctionCall(Exponent=@n20) {materialFunction=&AxisAlignedFresnel}.Fresnel) {constExponent=5}
  )
  n22 = Multiply(
   A=Add(A=Add(A=@n15, B=@n17) {desc=Adding some gradient for masking}, B=Constant() {desc=Tweaking texture brightness, r=0.25}) {constB=0.25}
   B=Add(A=@n19, B=Add(A=@n18.Fresnel, B=@n21))
  )
  n23 = ScalarParameter() {defaultValue=1.2, desc=0.6, parameterName=Texture Subtract}
  Saturate(None=Divide(A=Subtract(A=@n22, B=@n23), B=ScalarParameter() {defaultValue=0.05, parameterName=Texture Divide}))
 ]
}
